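# Landscape material function for a mossy ground layer in the Asteria environment, using Zen Garden textures.
# It samples color, mask and normal textures through Landscape_TextureSample, with scaled landscape coordinates and shared hex-grid controls for randomized scale, offset and rotation.
# Results set BaseColor and Normal, while the mask's red channel drives both Specular and Roughness.
# It outputs these material attributes plus a Height taken from the color sample's alpha.
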MaterialFunction /Game/Athena/Asteria/Environments/Landscape/Materials/MaterialFunctions/MF_MossFloor
graph {nodes [
  n1 = Multiply(A=LandscapeLayerCoords(), B=NamedRerouteUsage())
  n2 = MaterialFunctionCall!(
   Texture=TextureObject() {texture=/Game/Environments/Asteria/Sets/ZenGarden/Textures/T_Asteria_Zen_MossyGround_02}
   LandscapeCoordinates=@n1
   LowEndTiling
   HexGridGradientSharpness=NamedRerouteUsage()
   HexGridSize=NamedRerouteUsage()
   HexGridUVInput
   MinScale=NamedRerouteUsage()
   MaxScale=NamedRerouteUsage()
   MinOffset=NamedRerouteUsage()
   MaxOffset=NamedRerouteUsage()
   "MaxRotation(Degrees)"=NamedRerouteUsage()
   "MinRotation(Degrees)"=NamedRerouteUsage()
   forceLowQuality=NamedRerouteUsage()
  ) {materialFunction=/Game/Athena/Asteria/Environments/Landscape/Materials/MaterialFunctions/UtilityFunctions/Landscape_TextureSample}
  n3 = LandscapeLayerCoords()
  n4 = NamedRerouteUsage()
  n5 = MaterialFunctionCall(Float3=MaterialFunctionCall(
    Texture=TextureObject() {samplerType=SAMPLERTYPE_Masks, texture=/Game/Environments/Asteria/Sets/ZenGarden/Textures/T_Asteria_Zen_MossyGround_01_S}
    LandscapeCoordinates=Multiply(A=@n3, B=@n4)
    HexGridGradientSharpness=NamedRerouteUsage()
    HexGridSize=NamedRerouteUsage()
    MinScale=NamedRerouteUsage()
    MaxScale=NamedRerouteUsage()
    MinOffset=NamedRerouteUsage()
    MaxOffset=NamedRerouteUsage()
    "MaxRotation(Degrees)"=NamedRerouteUsage()
    "MinRotation(Degrees)"=NamedRerouteUsage()
    forceLowQuality=NamedRerouteUsage()
   ) {materialFunction=&Landscape_TextureSample}.Result) {materialFunction=&BreakOutFloat3Components}
  n6 = LandscapeLayerCoords()
  n7 = NamedRerouteUsage()
  n8 = SetMaterialAttributes!(
   MaterialAttributes
   BaseColor=@n2.Result
   Specular=@n5.R
   Roughness=@n5.R
   Normal=MaterialFunctionCall(
    Texture=TextureObject() {samplerType=SAMPLERTYPE_Normal, texture=/Game/Environments/Asteria/Sets/ZenGarden/Textures/T_Asteria_Zen_MossyGround_01_N}
    LandscapeCoordinates=Multiply(A=@n6, B=@n7)
    HexGridGradientSharpness=NamedRerouteUsage()
    HexGridSize=NamedRerouteUsage()
    MinScale=NamedRerouteUsage()
    MaxScale=NamedRerouteUsage()
    MinOffset=NamedRerouteUsage()
    MaxOffset=NamedRerouteUsage()
    "MaxRotation(Degrees)"=NamedRerouteUsage()
    "MinRotation(Degrees)"=NamedRerouteUsage()
    forceLowQuality=NamedRerouteUsage()
   ) {materialFunction=&Landscape_TextureSample}.Result
  ) {attributeSetTypes [
    #1
    #2
    #3
    #4
   ]}
  FunctionOutput(None=@n8) {outputName=Material}
  n9 = ComponentMask(None=@n2.Result) {a=true}
  FunctionOutput(None=@n9) {outputName=Height, sortPriority=1}
  Multiply(A=@n1)
  NamedRerouteDeclaration(None=ScalarParameter() {defaultValue=1, parameterName=HexGridSize}) {name=HexGridSize, nodeColor=<rgba 1 0.5391 0 1>}
  NamedRerouteDeclaration(None=ScalarParameter() {defaultValue=2, parameterName=HexGridSharpness}) {name=HexGridGradientSharpness, nodeColor=<rgba 0 0.7422 1 1>}
  NamedRerouteDeclaration(None=Constant() {r=1}) {name=MacroScale, nodeColor=<rgba 0 1 0.1328 1>}
  NamedRerouteDeclaration(None=Constant() {r=3}) {name=MipBiasValue, nodeColor=<rgba 0.6641 1 0 1>}
  NamedRerouteDeclaration(None=@n9) {name=MacroDetailHeight, nodeColor=<rgba 0.4531 1 0 1>}
  NamedRerouteDeclaration(None=ScalarParameter() {defaultValue=2, parameterName=MicroDetail_MinScale}) {name=MicroMinScale, nodeColor=<rgba 1 0.8906 0 1>}
  NamedRerouteDeclaration(None=ScalarParameter() {defaultValue=4, parameterName=MicroDetail_MaxScale}) {name=MicroDetailMaxScale, nodeColor=<rgba 1 0.8906 0 1>}
  NamedRerouteDeclaration(None=ScalarParameter() {defaultValue=4, parameterName=MinScale}) {name=MinScale, nodeColor=<rgba 1 0.8906 0 1>}
  NamedRerouteDeclaration(None=ScalarParameter() {defaultValue=8, parameterName=MaxScale}) {name=MaxScale, nodeColor=<rgba 1 0.8906 0 1>}
  NamedRerouteDeclaration(None=Constant() {r=-1}) {name=MinOffset, nodeColor=<rgba 0.7812 1 0 1>}
  NamedRerouteDeclaration(None=Constant() {r=1}) {name=MaxOffset, nodeColor=<rgba 0.7812 1 0 1>}
  NamedRerouteDeclaration(None=StaticBool()) {name=ForceLow, nodeColor=<rgba 0 1 0.3203 1>}
  NamedRerouteDeclaration(None=Constant()) {name=MinRotation, nodeColor=<rgba 0.5938 1 0 1>}
  NamedRerouteDeclaration(None=StaticBool()) {name=useMipBias, nodeColor=<rgba 0 1 0.9531 1>}
  NamedRerouteDeclaration(None=Constant()) {name=MaxRotation, nodeColor=<rgba 0.5938 1 0 1>}
  NamedRerouteUsage()
  NamedRerouteUsage()
 ]}
