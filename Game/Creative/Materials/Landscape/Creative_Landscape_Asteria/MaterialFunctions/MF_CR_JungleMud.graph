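# Landscape layer material function for Creative's Asteria jungle mud, outputting Material, Height and simplified MobileMaterial attributes.
# Hex-grid landscape sampling supplies slippery-mud color, masks and normals, with a detail normal layered on and diffuse alpha serving as height.
# A puddle mask built from sphere masking and upward-facing vertex normals blends in darker color and altered specular and roughness, and flattens normals.
# MobileMaterial uses constant-driven metallic, specular and roughness, avoiding texture samples.
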MaterialFunction /Game/Creative/Materials/Landscape/Creative_Landscape_Asteria/MaterialFunctions/MF_CR_JungleMud
graph {nodes [
  Constant() {r=1}
  n1 = LandscapeLayerCoords() {mappingRotation=75}
  n2 = NamedRerouteUsage()
  n3 = TextureSample(UVs=Divide(A=Multiply(A=@n1, B=@n2), B=ScalarParameter() {defaultValue=3, group=JungleMud, parameterName=JungleMud_PuddleDetailScale, sortPriority=200})) {
   samplerSource=SSM_Wrap_WorldGroupSettings
   samplerType=SAMPLERTYPE_Normal
   texture=/Game/Environments/Asteria/Terrain/Textures/T_Asteria_Mud_Slippery_01_N
  }
  n4 = Saturate(None=@n3.R)
  n5 = OneMinus(None=Saturate(None=Multiply(A=@n4) {constB=5}))
  n6 = MaterialFunctionCall(Input=@n5, Target Low=Constant() {r=0.75}, Target High=Constant() {r=0.9}) {materialFunction=&RemapValueRange}
  n7 = NamedRerouteUsage()
  n8 = MaterialFunctionCall(Float3=MaterialFunctionCall!(
    Texture=TextureObject() {samplerType=SAMPLERTYPE_Masks, texture=/Game/Environments/Asteria/Terrain/Textures/T_Asteria_Mud_Slippery_01_S}
    LandscapeCoordinates=NamedRerouteUsage()
    LowEndTiling=NamedRerouteUsage()
    HexGridGradientSharpness=NamedRerouteUsage()
    HexGridSize=NamedRerouteUsage()
    HexGridUVInput
    MinScale=NamedRerouteUsage()
    MaxScale=NamedRerouteUsage()
    MinOffset=NamedRerouteUsage()
    MaxOffset=NamedRerouteUsage()
    "MaxRotation(Degrees)"=NamedRerouteUsage()
    "MinRotation(Degrees)"=NamedRerouteUsage()
    forceLowQuality=NamedRerouteUsage()
   ) {materialFunction=/Game/Athena/Asteria/Environments/Landscape/Materials/MaterialFunctions/UtilityFunctions/Landscape_TextureSample}.Result) {materialFunction=&BreakOutFloat3Components}
  n9 = Constant() {r=0.03}
  n10 = Multiply(A=Saturate(None=Add(A=@n5) {constA=0.2, constB=0.2}), B=Constant() {r=0.032}) {constB=0.1}
  n11 = MaterialFunctionCall!(
   Texture=TextureObject() {samplerType=SAMPLERTYPE_Normal, texture=&T_Asteria_Mud_Slippery_01_N}
   UseMipBias=NamedRerouteUsage()
   MipBias=NamedRerouteUsage()
   LowEndTiling=NamedRerouteUsage()
   LandscapeCoordinates=NamedRerouteUsage()
   HexGridSize=NamedRerouteUsage()
   HexGridUVInput
   HexGridGradientSharpness=NamedRerouteUsage()
   MaxScale=NamedRerouteUsage()
   MinScale=NamedRerouteUsage()
   MaxOffset=NamedRerouteUsage()
   MinOffset=NamedRerouteUsage()
   "MaxRotation(Degrees)"=NamedRerouteUsage()
   "MinRotation(Degrees)"=NamedRerouteUsage()
   forceLowQuality=NamedRerouteUsage()
  ) {materialFunction=/Game/Athena/Asteria/Environments/Landscape/Materials/MaterialFunctions/UtilityFunctions/Landscape_NormalTextureSample}
  n12 = TextureSample(UVs=Divide(A=NamedRerouteUsage(), B=ScalarParameter() {defaultValue=1.2, group=JungleMud, parameterName=JungleMud_DetailNormalScale, sortPriority=100})) {
   samplerSource=SSM_Wrap_WorldGroupSettings
   samplerType=SAMPLERTYPE_Normal
   texture=&T_Asteria_Mud_Slippery_01_N
  }
  n13 = Constant3Vector() {constant=<rgba 0.33 0.33 0 1>}
  n14 = Normalize(VectorInput=Add(A=ComponentMask(None=@n11.Result) {b=true, g=true, r=true}, B=Multiply(A=@n12.RGB, B=@n13)))
  n15 = MaterialFunctionCall(Normal=@n14, Flatness=Constant() {r=1}) {materialFunction=&FlattenNormal}
  n16 = Saturate(None=NamedRerouteUsage())
  n17 = Constant() {r=0.8}
  n18 = Constant() {r=1}
  n19 = SetMaterialAttributes!(
   MaterialAttributes
   BaseColor=LinearInterpolate(A=NamedRerouteUsage(), B=Multiply(A=@n6.Result, B=@n7), Alpha=NamedRerouteUsage())
   Specular=LinearInterpolate(A=@n8.R, B=Constant() {r=0.25}, Alpha=NamedRerouteUsage())
   Roughness=LinearInterpolate(A=@n8.B, B=Add(A=@n9, B=@n10), Alpha=NamedRerouteUsage())
   Normal=Multiply(
    A=LinearInterpolate(A=@n14, B=@n15.Result, Alpha=@n16)
    B=MaterialFunctionCall(X=@n17, Y=@n17, Z=@n18) {materialFunction=&MakeFloat3}.Result
   )
  ) {attributeSetTypes [
    #1
    #2
    #3
    #4
   ]}
  FunctionOutput(None=@n19) {outputName=Material}
  n20 = LandscapeLayerCoords()
  n21 = NamedRerouteUsage()
  n22 = MaterialFunctionCall(
   Texture=TextureObject() {texture=/Game/Environments/Asteria/Terrain/Textures/T_Asteria_Mud_Slippery_01_D}
   LandscapeCoordinates=NamedRerouteDeclaration(None=Multiply(A=@n20, B=@n21)) {name=AdjustedLSCoords, nodeColor=<rgba 0 1 0.1328 1>}
   LowEndTiling=NamedRerouteUsage()
   HexGridGradientSharpness=NamedRerouteUsage()
   HexGridSize=NamedRerouteUsage()
   MinScale=NamedRerouteUsage()
   MaxScale=NamedRerouteUsage()
   MinOffset=NamedRerouteUsage()
   MaxOffset=NamedRerouteUsage()
   "MaxRotation(Degrees)"=NamedRerouteUsage()
   "MinRotation(Degrees)"=NamedRerouteUsage()
   forceLowQuality=NamedRerouteUsage()
  ) {materialFunction=&Landscape_TextureSample}
  FunctionOutput(None=NamedRerouteDeclaration(None=ComponentMask(None=@n22.Result) {a=true}) {name=BaseHeight, nodeColor=<rgba 1 0.5156 0 1>}) {outputName=Height, sortPriority=1}
  n23 = SetMaterialAttributes!(
   MaterialAttributes
   BaseColor
   Metallic=Constant()
   Specular=LinearInterpolate(A=Constant() {r=0.5}, B=Constant() {r=1})
   Roughness=LinearInterpolate(A=Constant() {r=0.1}, B=Constant() {r=0.25})
  ) {attributeSetTypes [
    #1
    #5
    #2
    #3
   ]}
  FunctionOutput(None=@n23) {outputName=MobileMaterial, sortPriority=2}
  NamedRerouteDeclaration(None=Constant() {r=5}) {name=HexGridSize, nodeColor=<rgba 1 0.5391 0 1>}
  NamedRerouteDeclaration(None=Constant() {r=8}) {name=HexGridGradientSharpness, nodeColor=<rgba 0 0.7422 1 1>}
  NamedRerouteDeclaration(None=ScalarParameter() {defaultValue=1, group=JungleMud, parameterName=JungleMud_OverallScale, sortPriority=-100}) {name=MacroScale, nodeColor=<rgba 0 1 0.1328 1>}
  NamedRerouteDeclaration(None=Constant() {r=3}) {name=MipBiasValue, nodeColor=<rgba 0.6641 1 0 1>}
  n24 = Constant() {r=0.1}
  n25 = Constant() {r=1.5}
  n26 = SphereMask(A=NamedRerouteUsage(), B=Constant() {r=0.05}, Radius=Constant() {r=0.25}, Hardness=Constant() {r=0.2})
  n27 = Multiply(A=NamedRerouteUsage(), B=Constant() {r=1})
  n28 = Constant() {r=-70}
  n29 = Constant() {r=1.2891}
  n30 = ComponentMask(None=VertexNormalWS()) {b=true}
  n31 = Multiply(A=Saturate(None=Subtract(A=@n26, B=@n27)), B=Saturate(None=LinearInterpolate(A=@n28, B=@n29, Alpha=@n30)))
  NamedRerouteDeclaration(None=Saturate(None=LinearInterpolate(A=@n24, B=@n25, Alpha=@n31))) {name=PuddleMask, nodeColor=<rgba 1 0.8438 0 1>}
  n32 = NamedRerouteUsage()
  n33 = Constant() {r=0.25}
  n34 = OneMinus(None=MaterialFunctionCall(Input=OneMinus(None=@n32), Input High=OneMinus(None=@n33)) {materialFunction=&RemapValueRange}.Result)
  n35 = Constant() {r=0.5}
  n36 = Multiply(
   A=LinearInterpolate(
    A=Constant3Vector() {constant=<rgba 0.5938 0.5938 0.5938 1>}
    B=Constant3Vector() {constant=<rgba 0.526 0.526 0.2575 1>}
    Alpha=Power(Base=@n34, Exp=@n35)
   )
   B=ComponentMask(None=@n22.Result) {b=true, g=true, r=true}
  )
  n37 = Constant3Vector() {constant=<rgba 0.6042 0.4464 0.1263 1>}
  n38 = NamedRerouteUsage()
  n39 = Constant() {r=1}
  n40 = Constant() {r=0.5}
  n41 = Constant() {r=0.1}
  NamedRerouteDeclaration(None=LinearInterpolate(A=@n36, B=Multiply(A=@n36, B=@n37), Alpha=SphereMask(A=@n38, B=@n39, Radius=@n40, Hardness=@n41))) {name=BaseColor, nodeColor=<rgba 0.9922 0 1 1>}
  NamedRerouteDeclaration(None=Constant() {r=4}) {name=MinScale, nodeColor=<rgba 1 0.8906 0 1>}
  n42 = OneMinus(None=MaterialFunctionCall(Input=NamedRerouteUsage(), Input High=Constant() {r=0.25}) {materialFunction=&RemapValueRange}.Result)
  n43 = Constant() {r=2}
  NamedRerouteDeclaration(None=LinearInterpolate(
    A=Constant3Vector() {constant=<rgba 0.1094 0.0552 0.0308 1>}
    B=Constant3Vector() {constant=<rgba 0.1042 0.0507 0.0403 1>}
    Alpha=Power(Base=@n42, Exp=@n43)
   )) {name=PuddleColor, nodeColor=<rgba 1 0 0.6562 1>}
  NamedRerouteDeclaration(None=Constant() {r=0.15}) {name=LowEndTiling, nodeColor=<rgba 0 0.0156 1 1>}
  NamedRerouteDeclaration(None=Constant() {r=9}) {name=MaxScale, nodeColor=<rgba 1 0.8906 0 1>}
  NamedRerouteDeclaration(None=Constant() {r=-1}) {name=MinOffset, nodeColor=<rgba 0.7812 1 0 1>}
  NamedRerouteDeclaration(None=Constant() {r=1}) {name=MaxOffset, nodeColor=<rgba 0.7812 1 0 1>}
  NamedRerouteDeclaration(None=StaticBool()) {name=ForceLow, nodeColor=<rgba 0 1 0.3203 1>}
  NamedRerouteDeclaration(None=Constant()) {name=MinRotation, nodeColor=<rgba 0.5938 1 0 1>}
  NamedRerouteDeclaration(None=StaticBool()) {name=useMipBias, nodeColor=<rgba 0 1 0.9531 1>}
  NamedRerouteDeclaration(None=Constant()) {name=MaxRotation, nodeColor=<rgba 0.5938 1 0 1>}
  NamedRerouteUsage() {desc=Exceeded texture samples on mobile}
 ]}
