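# Jungle landscape material for Creative's Asteria terrain (Chapter 4), driving surface shading, grass spawning and physical materials.
# Height-blends painted gravel, mud, sand, asphalt, jungle mud, floor and moss over base grass, carrying height in Opacity Mask, then applies cliffs.
# A ShadingPathSwitch selects a simplified mobile stack before shared gameplay and Lumen functions.
# Derived layer, cliff and cell-noise masks feed LandscapeGrassOutput and LandscapePhysicalMaterialOutput, with several grass slots zeroed.
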Material /BRCosmetics/Creative/Materials/Landscape/Creative_Landscape_Asteria/M_Landscape_Ch4_Jungle
graph {
 properties {
  bTangentSpaceNormal=false
  bUseMaterialAttributes=true
  bUsedWithNanite=true
  bUsedWithVirtualHeightfieldMesh=true
  pixelDepthOffsetMode=PDOM_Legacy
  shadingModel=MSM_FromMaterialExpression
  usedShadingModels=Default Lit
 }
 nodes [
  Constant() {r=0.75}
  Constant() {r=0.5}
  Constant() {r=0.75}
  MaterialExpressionMultiply_54 = Multiply(A=NamedRerouteUsage(), B=LandscapeVisibilityMask()) {desc=Grass_01}
  MaterialExpressionMultiply_53 = Multiply(A=NamedRerouteUsage(), B=LandscapeVisibilityMask()) {desc=Grass_02}
  MaterialExpressionMultiply_52 = Multiply(A=NamedRerouteUsage(), B=LandscapeVisibilityMask()) {desc=Grass_03}
  n1 = NamedRerouteUsage()
  n2 = NamedRerouteUsage()
  n3 = NamedRerouteUsage()
  n4 = NamedRerouteUsage()
  n5 = OneMinus(None=Max(A=Max(A=@n2, B=@n3), B=@n4))
  MaterialExpressionMultiply_51 = Multiply(A=Multiply(A=@n1, B=@n5), B=LandscapeVisibilityMask()) {desc=Gravel_01}
  n6 = NamedRerouteUsage()
  MaterialExpressionMultiply_50 = Multiply(A=Multiply(A=Multiply(A=@n1, B=@n6), B=@n5), B=LandscapeVisibilityMask()) {desc=Gravel_02}
  n7 = NamedRerouteUsage()
  n8 = NamedRerouteUsage()
  MaterialExpressionMultiply_49 = Multiply(A=Multiply(A=Multiply(A=@n7, B=@n8), B=@n5), B=LandscapeVisibilityMask()) {desc=Gravel_03}
  MaterialExpressionConstant_13 = Constant()
  n9 = Multiply(A=OneMinus(None=@n4), B=NamedRerouteUsage())
  n10 = Max(A=Max(A=NamedRerouteUsage(), B=NamedRerouteUsage()), B=@n9)
  MaterialExpressionMultiply_45 = Multiply(A=@n10, B=LandscapeVisibilityMask()) {desc=JungleFloor_01}
  MaterialExpressionMultiply_43 = Multiply(A=Multiply(A=@n10, B=NamedRerouteUsage()), B=LandscapeVisibilityMask()) {desc=JungleFloor_02}
  MaterialExpressionMultiply_42 = Multiply(A=@n9, B=LandscapeVisibilityMask()) {desc=JungleFloor_03}
  LandscapeGrassOutput!(
   Grass_01=@MaterialExpressionMultiply_54
   Grass_02=@MaterialExpressionMultiply_53
   Grass_03=@MaterialExpressionMultiply_52
   Gravel_01=@MaterialExpressionMultiply_51
   Gravel_02=@MaterialExpressionMultiply_50
   Gravel_03=@MaterialExpressionMultiply_49
   Scree=@MaterialExpressionConstant_13
   ForestFloor=@MaterialExpressionConstant_13
   FlowerField=@MaterialExpressionConstant_13
   BeachFoliage=@MaterialExpressionConstant_13
   JungleFloor_01=@MaterialExpressionMultiply_45
   JungleFloor_02=@MaterialExpressionMultiply_43
   JungleFloor_03=@MaterialExpressionMultiply_42
  ) {grassTypes [
    Grass_01 {
     grassType=/Game/Creative/Materials/Landscape/Creative_Landscape_Asteria/Grass/LGT_Creative_Asteria_Standard
     input {
      expression=&self:MaterialExpressionMultiply_54
      inputName=None
      mask=0
      maskA=0
      maskB=0
      maskG=0
      maskR=0
      outputIndex=0
     }
    }
    Grass_02 {
     grassType=/Game/Creative/Materials/Landscape/Creative_Landscape_Asteria/Grass/LGT_Creative_Asteria_Standard_Variation_01
     input {
      expression=&self:MaterialExpressionMultiply_53
      inputName=None
      mask=0
      maskA=0
      maskB=0
      maskG=0
      maskR=0
      outputIndex=0
     }
    }
    Grass_03 {
     grassType=/Game/Creative/Materials/Landscape/Creative_Landscape_Asteria/Grass/LGT_Creative_Asteria_Standard_Variation_02
     input {
      expression=&self:MaterialExpressionMultiply_52
      inputName=None
      mask=0
      maskA=0
      maskB=0
      maskG=0
      maskR=0
      outputIndex=0
     }
    }
    Gravel_01 {
     grassType=/Game/Creative/Materials/Landscape/Creative_Landscape_Asteria/Grass/LGT_Gravel_V1
     input {
      expression=&self:MaterialExpressionMultiply_51
      inputName=None
      mask=0
      maskA=0
      maskB=0
      maskG=0
      maskR=0
      outputIndex=0
     }
    }
    Gravel_02 {
     grassType=/Game/Creative/Materials/Landscape/Creative_Landscape_Asteria/Grass/LGT_Gravel_V2
     input {
      expression=&self:MaterialExpressionMultiply_50
      inputName=None
      mask=0
      maskA=0
      maskB=0
      maskG=0
      maskR=0
      outputIndex=0
     }
    }
    Gravel_03 {
     grassType=/Game/Creative/Materials/Landscape/Creative_Landscape_Asteria/Grass/LGT_Gravel_V3
     input {
      expression=&self:MaterialExpressionMultiply_49
      inputName=None
      mask=0
      maskA=0
      maskB=0
      maskG=0
      maskR=0
      outputIndex=0
     }
    }
    Scree {
     grassType=/Game/Creative/Materials/Landscape/Creative_Landscape_Asteria/Grass/LGT_Creative_Scree
     input {
      expression=&self:MaterialExpressionConstant_13
      inputName=None
      mask=0
      maskA=0
      maskB=0
      maskG=0
      maskR=0
      outputIndex=0
     }
    }
    ForestFloor {
     grassType=/Game/Creative/Materials/Landscape/Creative_Landscape_Asteria/Grass/LGT_Creative_ForestFloor
     input {
      expression=&self:MaterialExpressionConstant_13
      inputName=None
      mask=0
      maskA=0
      maskB=0
      maskG=0
      maskR=0
      outputIndex=0
     }
    }
    FlowerField {
     grassType=/Game/Creative/Materials/Landscape/Creative_Landscape_Asteria/Grass/LGT_FlowerField
     input {
      expression=&self:MaterialExpressionConstant_13
      inputName=None
      mask=0
      maskA=0
      maskB=0
      maskG=0
      maskR=0
      outputIndex=0
     }
    }
    BeachFoliage {
     grassType=/Game/Creative/Materials/Landscape/Creative_Landscape_Asteria/Grass/LGT_BeachFoliage_01
     input {
      expression=&self:MaterialExpressionConstant_13
      inputName=None
      mask=0
      maskA=0
      maskB=0
      maskG=0
      maskR=0
      outputIndex=0
     }
    }
    JungleFloor_01 {
     grassType=/Game/Creative/Materials/Landscape/Creative_Landscape_Asteria/Grass/LGT_JungleFloor_01
     input {
      expression=&self:MaterialExpressionMultiply_45
      inputName=None
      mask=0
      maskA=0
      maskB=0
      maskG=0
      maskR=0
      outputIndex=0
     }
    }
    JungleFloor_02 {
     grassType=/Game/Environments/Asteria/Foliage/Grass/Spawners/LS_JungleFloor_02
     input {
      expression=&self:MaterialExpressionMultiply_43
      inputName=None
      mask=0
      maskA=0
      maskB=0
      maskG=0
      maskR=0
      outputIndex=0
     }
    }
    JungleFloor_03 {
     grassType=/Game/Environments/Asteria/Foliage/Grass/Spawners/LS_JungleFloor_03
     input {
      expression=&self:MaterialExpressionMultiply_42
      inputName=None
      mask=0
      maskA=0
      maskB=0
      maskG=0
      maskR=0
      outputIndex=0
     }
    }
   ]}
  MaterialExpressionNamedRerouteUsage_92 = NamedRerouteUsage()
  MaterialExpressionNamedRerouteUsage_96 = NamedRerouteUsage()
  MaterialExpressionNamedRerouteUsage_98 = NamedRerouteUsage()
  MaterialExpressionNamedRerouteUsage_99 = NamedRerouteUsage()
  MaterialExpressionNamedRerouteUsage_94 = NamedRerouteUsage()
  MaterialExpressionConstant_31 = Constant()
  MaterialExpressionConstant_4 = Constant()
  LandscapePhysicalMaterialOutput!(
   Dirt=@MaterialExpressionNamedRerouteUsage_92
   Stone=@MaterialExpressionNamedRerouteUsage_96
   Grass=@MaterialExpressionNamedRerouteUsage_98
   Asphalt=@MaterialExpressionNamedRerouteUsage_99
   Sand=@MaterialExpressionNamedRerouteUsage_94
   PhysData_Ice=@MaterialExpressionConstant_31
   Snow=@MaterialExpressionConstant_4
  ) {
   desc="These are layered in order of importance.\r\nIf Dirt=1 and Stone=1, then Dirt will \"win\".\r\nTheoretically, these values aren't clamped.\r\nSo a value of 2 should trump a value of 1."
   inputs [
    {
     physicalMaterial=/Game/FortressPhysicalMaterials/Dirt
     input {
      expression=&self:MaterialExpressionNamedRerouteUsage_92
      inputName=None
      mask=0
      maskA=0
      maskB=0
      maskG=0
      maskR=0
      outputIndex=0
     }
    }
    {
     physicalMaterial=/Game/FortressPhysicalMaterials/Stone
     input {
      expression=&self:MaterialExpressionNamedRerouteUsage_96
      inputName=None
      mask=0
      maskA=0
      maskB=0
      maskG=0
      maskR=0
      outputIndex=0
     }
    }
    {
     physicalMaterial=/Game/FortressPhysicalMaterials/Grass
     input {
      expression=&self:MaterialExpressionNamedRerouteUsage_98
      inputName=None
      mask=0
      maskA=0
      maskB=0
      maskG=0
      maskR=0
      outputIndex=0
     }
    }
    {
     physicalMaterial=/Game/FortressPhysicalMaterials/Asphalt
     input {
      expression=&self:MaterialExpressionNamedRerouteUsage_99
      inputName=None
      mask=0
      maskA=0
      maskB=0
      maskG=0
      maskR=0
      outputIndex=0
     }
    }
    {
     physicalMaterial=/Game/FortressPhysicalMaterials/Sand
     input {
      expression=&self:MaterialExpressionNamedRerouteUsage_94
      inputName=None
      mask=0
      maskA=0
      maskB=0
      maskG=0
      maskR=0
      outputIndex=0
     }
    }
    {
     physicalMaterial=/Game/Balance/Physics/Materials/PhysData_Ice
     input {
      expression=&self:MaterialExpressionConstant_31
      inputName=None
      mask=0
      maskA=0
      maskB=0
      maskG=0
      maskR=0
      outputIndex=0
     }
    }
    {
     physicalMaterial=/Game/FortressPhysicalMaterials/Snow
     input {
      expression=&self:MaterialExpressionConstant_4
      inputName=None
      mask=0
      maskA=0
      maskB=0
      maskG=0
      maskR=0
      outputIndex=0
     }
    }
   ]
  }
  n11 = LinearInterpolate(A=NamedRerouteUsage(), B=NamedRerouteUsage(), Alpha=ScalarParameter() {parameterName=DEBUG-PhysMat1})
  n12 = NamedRerouteUsage()
  n13 = ScalarParameter() {parameterName=DEBUG-PhysMat2}
  LinearInterpolate(
   A=LinearInterpolate(A=LinearInterpolate(A=@n11, B=@n12, Alpha=@n13), B=NamedRerouteUsage(), Alpha=ScalarParameter() {parameterName=DEBUG-PhysMat3})
   B=NamedRerouteUsage()
   Alpha=ScalarParameter() {parameterName=DEBUG-PhysMat4}
  )
  Multiply(A=Max(A=NamedRerouteUsage(), B=NamedRerouteUsage()), B=NamedRerouteUsage())
  n14 = NamedRerouteUsage()
  n15 = NamedRerouteDeclaration(None=Multiply(A=NamedRerouteUsage(), B=OneMinus(None=@n14))) {name=Sample_BiomeC-FlowerFields, nodeColor=<rgba 0 1 0.7188 1>}
  n16 = NamedRerouteUsage()
  Multiply(A=@n15, B=OneMinus(None=@n16))
  n17 = NamedRerouteUsage()
  n18 = MaterialFunctionCall() {materialFunction=/Game/Creative/Materials/Landscape/Creative_Landscape_Asteria/MaterialFunctions/MF_CR_JungleMud}
  n19 = Saturate(None=@n18.Height)
  n20 = MaterialFunctionCall!(
   CurrentMaterial=@n17
   AddMaterial=SetMaterialAttributes!(MaterialAttributes=@n18.Material, Opacity Mask=@n19) {attributeSetTypes [#1]}
   LandscapeLayer=NamedRerouteDeclaration(None=LandscapeLayerSample() {parameterName=Layer5}) {name=Layer5_JungleMud_Scree, nodeColor=<rgba 1 0.5859 0 1>}
   UseContrast
   Contrast
  ) {materialFunction=/Game/Creative/Materials/Landscape/Creative_Landscape_Apollo/MaterialFunctions/MF_BlendLandscapeLayers_Opacity}
  NamedRerouteDeclaration(None=LandscapeLayerSwitch!(
    LayerUsed=SetMaterialAttributes(MaterialAttributes=@n20.BlendedMaterial, Opacity Mask=@n20.BlendedMaterial) {attributeSetTypes [#1]}
    LayerNotUsed=@n17
   ) {parameterName=Layer5}) {name=With_BiomeA, nodeColor=<rgba 1 1 0 1>}
  n21 = NamedRerouteUsage()
  n22 = NamedRerouteUsage()
  n23 = Max(A=Max(A=Max(A=@n21, B=@n22), B=NamedRerouteUsage()), B=NamedRerouteUsage())
  n24 = ComponentMask(None=VertexNormalWS()) {b=true}
  n25 = OneMinus(None=Power(Base=Abs(None=@n24), Exp=Constant() {r=3}))
  n26 = Constant() {r=0.3}
  n27 = Max(A=Max(A=@n23, B=Multiply(A=@n25, B=@n26)), B=NamedRerouteUsage())
  n28 = Constant() {r=2}
  n29 = MaterialFunctionCall!(
   BiomeMask-RGBA=NamedRerouteUsage()
   Grass Localised Tint Mask
   CombinedLandscapeLayersMask=Saturate(None=Multiply(A=@n27, B=@n28))
   BiomeColors-UseDefault=StaticBool()
  ) {materialFunction=/Game/Athena/Asteria/Environments/Landscape/Materials/MaterialFunctions/MF_Asteria_Common_Grass}
  NamedRerouteDeclaration(None=@n29.GrassColorMask1) {name=GrassColorMask1, nodeColor=<rgba 0.5 0 1 1>}
  n30 = NamedRerouteUsage()
  n31 = TextureObject() {texture=/Game/Environments/Asteria/Terrain/Textures/T_Asteria_JungleFloor_D}
  n32 = MaterialFunctionCall!(
   LowEndTiling
   UseHexGridonLow
   BaseCoordinates
   MacroScale=Constant() {r=1}
   MinScale
   Diffuse=@n31
   MaxScale
   Normal=TextureObject() {samplerType=SAMPLERTYPE_Normal, texture=/Game/Environments/Asteria/Terrain/Textures/T_Asteria_JungleFloor_N}
   S_Mask=TextureObject() {samplerType=SAMPLERTYPE_Masks, texture=/Game/Environments/Asteria/Terrain/Textures/T_Asteria_JungleFloor_S}
   HexGridSize
   HexGridGradientShaprmness
   MinOffset
   MaxOffset
   MinRotation
   MaxRotation
   UseRegularTiling=StaticBool()
   MipBiasValue
   UseMipBias
  ) {materialFunction=/Game/Athena/Asteria/Environments/Landscape/Materials/MaterialFunctions/MF_Biome_Generic_v2}
  n33 = NamedRerouteDeclaration(None=LandscapeLayerSample() {parameterName=Layer6}) {name=Layer6_JungleFloor_ForestFloor, nodeColor=<rgba 0.0547 1 0 1>}
  n34 = StaticBool() {value=true}
  n35 = Constant() {r=0.5}
  n36 = MaterialFunctionCall(
   CurrentMaterial=@n30
   AddMaterial=SetMaterialAttributes(MaterialAttributes=@n32.Result, Opacity Mask=@n32.Result) {attributeSetTypes [#1]}
   LandscapeLayer=@n33
   UseContrast=@n34
   Contrast=@n35
  ) {materialFunction=&MF_BlendLandscapeLayers_Opacity}
  NamedRerouteDeclaration(None=LandscapeLayerSwitch(
    LayerUsed=SetMaterialAttributes(MaterialAttributes=@n36.BlendedMaterial, Opacity Mask=@n36.BlendedMaterial) {attributeSetTypes [#1]}
    LayerNotUsed=@n30
   ) {parameterName=Layer6}) {name=With_BiomeB, nodeColor=<rgba 1 1 0 1>}
  n37 = NamedRerouteUsage()
  n38 = MaterialFunctionCall() {materialFunction=/Game/Athena/Asteria/Environments/Landscape/Materials/MaterialFunctions/MF_MossFloor}
  n39 = NamedRerouteDeclaration(None=LandscapeLayerSample() {parameterName=Layer7}) {name=Layer7_JungleMoss_BiomeC, nodeColor=<rgba 1 0 0.375 1>}
  n40 = StaticBool() {value=true}
  n41 = Constant() {r=1}
  n42 = MaterialFunctionCall(
   CurrentMaterial=@n37
   AddMaterial=SetMaterialAttributes(MaterialAttributes=@n38.Material, Opacity Mask=@n38.Material) {attributeSetTypes [#1]}
   LandscapeLayer=@n39
   UseContrast=@n40
   Contrast=@n41
  ) {materialFunction=&MF_BlendLandscapeLayers_Opacity}
  n43 = GetMaterialAttributes(None=LandscapeLayerSwitch(
    LayerUsed=SetMaterialAttributes(MaterialAttributes=@n42.BlendedMaterial, Opacity Mask=@n42.BlendedMaterial) {attributeSetTypes [#1]}
    LayerNotUsed=@n37
   ) {parameterName=Layer7}) {attributeGetTypes [#1]}
  n44 = MaterialFunctionCall!(
   Material=@n43.MaterialAttributes
   Heightmap=@n43.MaterialAttributes
   CliffSnowMask
   NewSlopeMaskMask
   AltCliffSlopeScreeMask
   AsphaltLayer
   UseAlternateCliffMat
   AltCliffHeight
   AltCliffMat
  ) {materialFunction=/Game/Creative/Materials/Landscape/Creative_Landscape_Asteria/MaterialFunctions/MF_CR_AsteriaCliffBlend}
  NamedRerouteDeclaration(None=@n44.SlopeScreeMask) {name=SlopeScreeMask, nodeColor=<rgba 0.5 0 1 1>}
  n45 = NamedRerouteDeclaration(None=LandscapeLayerSample() {parameterName=Layer1}) {name=Layer1_Gravel, nodeColor=<rgba 0.2891 1 0 1>}
  n46 = MaterialFunctionCall!(LayerWeight=@n45, VarianceMask=NamedRerouteUsage(), BiomeMask-RGBA, BiomeColors-UseDefault=StaticBool()) {materialFunction=/Game/Creative/Materials/Landscape/Creative_Landscape_Asteria/MaterialFunctions/MF_CR_AsteriaGravel}
  n47 = MaterialFunctionCall!(
   Material 1=@n29.Default-LandscapeMaterial
   Height Texture 1=@n29.Default-LandscapeMaterial
   Material 2=@n46.GravelOnly
   Height Texture 2=@n46.GravelOnly
   Transistion Phase=@n45
   UseContrast
   Contrast
   AddLayerMask
  ) {materialFunction=/Game/Athena/Apollo/Environments/Landscape/Materials/Biome/CommonMaterials/MF_BlendLandscapeLayers}
  n48 = MaterialFunctionCall!(
   Spec=@n46.Spec
   Diffuse=@n46.Spec
   Height=@n46.Spec
   BiomeColors-UseDefault=StaticBool()
   Rough=@n46.Spec
   Normals=@n46.Spec
   VarianceMask=NamedRerouteUsage()
   BiomeMask-RGBA
   BiomeRoughMod
  ) {materialFunction=/Game/Athena/Asteria/Environments/Landscape/Materials/MaterialFunctions/MF_Asteria_Mud}
  n49 = MaterialFunctionCall(
   Material 1=@n47.BlendedMaterial
   Height Texture 1=@n47.BlendedMaterial
   Material 2=@n48.Material
   Height Texture 2=@n48.Material
   Transistion Phase=NamedRerouteDeclaration(None=LandscapeLayerSample() {parameterName=Layer2}) {name=Layer2_Mud, nodeColor=<rgba 1 0 0.6094 1>}
  ) {materialFunction=&MF_BlendLandscapeLayers}
  NamedRerouteDeclaration(None=SetMaterialAttributes(MaterialAttributes=@n49.BlendedMaterial, Opacity Mask=@n49.BlendedMaterial) {attributeSetTypes [#1]}) {name=With_Mud, nodeColor=<rgba 1 1 0 1>}
  n50 = Max(A=NamedRerouteUsage(), B=NamedRerouteUsage())
  n51 = NamedRerouteUsage()
  n52 = Max(A=Max(A=NamedRerouteUsage(), B=Max(A=@n50, B=@n51)), B=NamedRerouteUsage())
  n53 = NamedRerouteUsage()
  n54 = NamedRerouteUsage()
  n55 = OneMinus(None=Max(A=@n52, B=Max(A=@n53, B=@n54)))
  NamedRerouteDeclaration(None=@n55) {name=GM-Grass_01, nodeColor=<rgba 0.5 0 1 1>}
  NamedRerouteDeclaration(None=@n55) {name=GM-Bush, nodeColor=<rgba 0.5 0 1 1>}
  n56 = NamedRerouteDeclaration(None=LandscapeLayerSample() {parameterName=Layer3}) {name=Layer3_Sand, nodeColor=<rgba 0.8281 0 1 1>}
  n57 = NamedRerouteUsage()
  n58 = MaterialFunctionCall!(UseUnderOceanSandMaterial, BypassBiomeColors=StaticBool() {value=true}, UseOceanLevel) {materialFunction=/Game/Athena/Apollo/Environments/Landscape/Materials/Biome/CommonMaterials/MF_Common_Sand}
  n59 = MaterialFunctionCall(
   CurrentMaterial=@n57
   AddMaterial=SetMaterialAttributes(MaterialAttributes=@n58.Material, Opacity Mask=@n58.Material) {attributeSetTypes [#1]}
   LandscapeLayer=@n56
  ) {materialFunction=&MF_BlendLandscapeLayers_Opacity}
  n60 = Saturate(None=LinearInterpolate(A=Constant() {r=30}, B=Constant(), Alpha=@n59.Alpha))
  NamedRerouteDeclaration(None=Saturate(None=Add(A=@n56, B=@n60))) {name=M_Sand, nodeColor=<rgba 0.5 0 1 1>}
  n61 = NamedRerouteUsage()
  n62 = NamedRerouteUsage()
  n63 = NamedRerouteUsage()
  n64 = Constant() {r=5}
  n65 = Constant() {r=-1}
  n66 = NamedRerouteUsage()
  n67 = MaterialFunctionCall!(TilingSize) {materialFunction=/Game/Athena/Apollo/Environments/Landscape/Materials/MaterialFunctions/MF_ApolloAsphalt}
  n68 = MaterialFunctionCall(
   CurrentMaterial=@n66
   AddMaterial=SetMaterialAttributes(MaterialAttributes=@n67.DefaultLandscapeMaterial, Opacity Mask=@n67.DefaultLandscapeMaterial) {attributeSetTypes [#1]}
   LandscapeLayer=NamedRerouteDeclaration(None=LandscapeLayerSample() {parameterName=Layer4}) {name=Layer4_Asphalt, nodeColor=<rgba 0 0.1328 1 1>}
  ) {materialFunction=&MF_BlendLandscapeLayers_Opacity}
  n69 = NamedRerouteDeclaration(None=Saturate(None=LinearInterpolate(A=@n64, B=@n65, Alpha=@n68.Alpha))) {name=AsphaltMask, nodeColor=<rgba 0.5 0 1 1>}
  NamedRerouteDeclaration(None=Max(A=Max(A=@n61, B=@n62), B=Max(A=@n63, B=@n69))) {name=GM-CommonLayers, nodeColor=<rgba 0.5 0 1 1>}
  NamedRerouteDeclaration(None=LandscapeLayerSwitch(
    LayerUsed=SetMaterialAttributes(MaterialAttributes=@n68.BlendedMaterial, Opacity Mask=@n68.BlendedMaterial) {attributeSetTypes [#1]}
    LayerNotUsed=@n66
   ) {parameterName=Layer4}) {name=With_Asphalt, nodeColor=<rgba 1 1 0 1>}
  n70 = MaterialFunctionCall!(
   Texture=TextureObject() {samplerType=SAMPLERTYPE_Grayscale, texture=/BRCosmetics/Weapons/FORT_Melee/Pickaxe_Cube_Ninja_Male/FX/T_CellNoise}
   LandscapeCoordinates=Multiply(A=Constant() {r=0.05}, B=LandscapeLayerCoords())
   LowEndTiling
   HexGridGradientSharpness
   HexGridSize
   HexGridUVInput
   MinScale
   MaxScale
   MinOffset
   MaxOffset
   "MaxRotation(Degrees)"
   "MinRotation(Degrees)"
   forceLowQuality
  ) {materialFunction=/Game/Athena/Asteria/Environments/Landscape/Materials/MaterialFunctions/UtilityFunctions/Landscape_TextureSample}
  n71 = Power(Base=Multiply(A=ComponentMask(None=@n70.Result) {r=true}, B=Constant() {r=10}), Exp=Constant() {r=4})
  NamedRerouteDeclaration(None=Multiply(A=Clamp(None=@n71), B=NamedRerouteUsage())) {name=GM-VariationMask-1, nodeColor=<rgba 1 0 0.8203 1>}
  n72 = Constant() {r=0.025}
  n73 = LandscapeLayerCoords()
  n74 = ComponentMask(None=MaterialFunctionCall(
    Texture=TextureObject() {samplerType=SAMPLERTYPE_Grayscale, texture=&T_CellNoise}
    LandscapeCoordinates=Multiply(A=@n72, B=@n73)
   ) {materialFunction=&Landscape_TextureSample}.Result) {r=true}
  n75 = Constant() {r=2}
  n76 = Clamp(None=Power(Base=Multiply(A=@n74, B=@n75), Exp=Constant() {r=7}))
  NamedRerouteDeclaration(None=Multiply(A=@n76, B=NamedRerouteUsage())) {name=GM-VariationMask-2, nodeColor=<rgba 0 1 0.8125 1>}
  NamedRerouteDeclaration(None=OneMinus(None=@n52)) {name=GM-CombinedCommonLayers, nodeColor=<rgba 0.2891 0 1 1>}
  n77 = TextureObject() {samplerType=SAMPLERTYPE_Grayscale, texture=&T_CellNoise}
  n78 = Multiply(A=Constant() {r=0.2}, B=LandscapeLayerCoords())
  n79 = Multiply(
   A=ComponentMask(None=MaterialFunctionCall(Texture=@n77, LandscapeCoordinates=@n78) {materialFunction=&Landscape_TextureSample}.Result) {r=true}
   B=Constant() {r=10}
  )
  n80 = Constant() {r=4}
  NamedRerouteDeclaration(None=Clamp(None=Power(Base=@n79, Exp=@n80))) {name=VariationMask_01, nodeColor=<rgba 0 1 0.5781 1>}
  n81 = TextureObject() {samplerType=SAMPLERTYPE_Grayscale, texture=&T_CellNoise}
  n82 = Multiply(A=Constant() {r=0.075}, B=LandscapeLayerCoords())
  n83 = Multiply(
   A=ComponentMask(None=MaterialFunctionCall(Texture=@n81, LandscapeCoordinates=@n82) {materialFunction=&Landscape_TextureSample}.Result) {r=true}
   B=Constant() {r=2}
  )
  n84 = Constant() {r=4}
  NamedRerouteDeclaration(None=Clamp(None=Power(Base=@n83, Exp=@n84))) {name=VariationMask_02, nodeColor=<rgba 1 0 0.6562 1>}
  n85 = NamedRerouteUsage()
  n86 = NamedRerouteUsage()
  n87 = MaterialFunctionCall!(OverrideEdgeColor, BiomeMask-RGBA, EdgeMask=@n86, BypassBiomeColors=StaticBool() {value=true}) {materialFunction=/Game/Athena/Asteria/Environments/Landscape/Materials/MaterialFunctions/MobileLayers/MF_Mobile_Asteria_Gravel}
  n88 = MaterialFunctionCall!(
   BiomeMask-RGBA=NamedRerouteUsage()
   GravelMaterialInput=@n85
   Heightmap=NamedRerouteUsage()
   UseUnmodifiedGravelDiffuse=StaticBool() {value=true}
   UnmodifiedGravelDiffuse=@n87.RGB
   BypassBiomeColors=StaticBool() {value=true}
  ) {materialFunction=/Game/Athena/Asteria/Environments/Landscape/Materials/MaterialFunctions/MobileLayers/MF_Asteria_DirtMudModifier}
  NamedRerouteDeclaration(None=SetMaterialAttributes!(MaterialAttributes=NamedRerouteUsage(), Base Color=@n88.ColorizedGravelDiffuse_UEFN, Opacity Mask=NamedRerouteUsage()) {attributeSetTypes [#2, #1]}) {name=MobileJungleMudMat, nodeColor=<rgba 1 1 0 1>}
  n89 = Constant() {r=5}
  n90 = Constant() {r=-1}
  NamedRerouteDeclaration(None=Saturate(None=LinearInterpolate(A=@n89, B=@n90, Alpha=@n47.Alpha))) {name=GravelMask, nodeColor=<rgba 0.5 0 1 1>}
  NamedRerouteDeclaration(None=@n76) {name=GM-VariationMask-2_1, nodeColor=<rgba 0 1 0.8125 1>}
  n91 = Constant() {r=5}
  n92 = Constant() {r=-1}
  NamedRerouteDeclaration(None=Saturate(None=LinearInterpolate(A=@n91, B=@n92, Alpha=@n49.Alpha))) {name=MudMask, nodeColor=<rgba 0.5 0 1 1>}
  n93 = NamedRerouteUsage()
  n94 = NamedRerouteUsage()
  n95 = Max(A=Max(A=Max(A=@n93, B=@n94), B=NamedRerouteUsage()), B=NamedRerouteUsage())
  NamedRerouteDeclaration(None=Multiply(A=NamedRerouteUsage(), B=OneMinus(None=@n95))) {name=GM-Gravel, nodeColor=<rgba 0 0.2734 1 1>}
  n96 = OneMinus(None=@n42.Alpha)
  NamedRerouteDeclaration(None=Multiply(A=Power(Base=@n96) {constExponent=10}, B=Power(Base=@n39) {constExponent=3})) {name=Layer7_MossFloor_SpawnMask, nodeColor=<rgba 0.6172 1 0 1>}
  NamedRerouteDeclaration(None=Constant()) {name=Layer-Snow, nodeColor=<rgba 1 1 1 1>}
  NamedRerouteDeclaration(None=Multiply(A=NamedRerouteUsage(), B=NamedRerouteUsage())) {name=GM-Snow, nodeColor=<rgba 0.5234 0 1 1>}
  n97 = Max(A=@n52, B=@n54)
  NamedRerouteDeclaration(None=Multiply(A=@n53, B=OneMinus(None=@n97))) {name=GM-Grass_02, nodeColor=<rgba 0.7109 0 1 1>}
  NamedRerouteDeclaration(None=Multiply(A=OneMinus(None=@n52), B=@n54)) {name=GM-Grass_03, nodeColor=<rgba 1 0 0.6094 1>}
  NamedRerouteDeclaration(None=OneMinus(None=@n20.Alpha)) {name=ScreeMask, nodeColor=<rgba 1 0 0.6797 1>}
  NamedRerouteDeclaration(None=@n32.Height) {name=JungleFloor-Height, nodeColor=<rgba 0.6928 0.156 1 1>}
  n98 = OneMinus(None=NamedRerouteUsage())
  n99 = Constant() {r=-0.5}
  n100 = Constant() {r=1.5}
  n101 = NamedRerouteUsage()
  n102 = MaterialFunctionCall(
   Transition Phase=Saturate(None=LinearInterpolate(A=@n99, B=@n100, Alpha=@n101))
   Height Texture=NamedRerouteUsage()
   Contrast=Constant() {r=0.5}
  ) {materialFunction=&HeightLerp}
  NamedRerouteDeclaration(None=Max(A=Multiply(A=@n98, B=@n102.Alpha), B=NamedRerouteUsage())) {name=GM-Scree, nodeColor=<rgba 1 0 0.3281 1>}
  NamedRerouteDeclaration(None=@n44.CliffEmissiveMask) {name=CliffEmissiveMask, nodeColor=<rgba 1 0 0.3281 1>}
  NamedRerouteDeclaration(None=@n19) {name=JungleMudHeight, nodeColor=<rgba 0.6928 0.156 1 1>}
  NamedRerouteDeclaration(None=Multiply(A=OneMinus(None=@n36.Alpha), B=@n33)) {name=Layer6_JungleFloor_Opacity, nodeColor=<rgba 1 0 0.1406 1>}
  NamedRerouteDeclaration(None=Constant()) {name=BiomeMask-Tundra, nodeColor=<rgba 0 0.4609 1 1>}
  NamedRerouteDeclaration(None=Constant()) {name=BiomeMask-Fall, nodeColor=<rgba 0 1 0 1>}
  NamedRerouteDeclaration(None=Multiply(A=NamedRerouteUsage(), B=@n55)) {name=GM-Fall-Grass, desc=NOT USED, nodeColor=<rgba 0 1 0.6953 1>}
  NamedRerouteDeclaration(None=Multiply(A=@n15, B=@n16)) {name=GM-FallFlowerField, desc=NOT USED, nodeColor=<rgba 0 1 0.8828 1>}
  NamedRerouteDeclaration(None=@n15) {name=GM-FlowerField, nodeColor=<rgba 1 0.9609 0 1>}
  NamedRerouteDeclaration(None=LandscapeLayerSwitch(
    LayerUsed=SetMaterialAttributes!(MaterialAttributes=@n59.BlendedMaterial, Opacity Mask=@n59.BlendedMaterial) {attributeSetTypes [#1]}
    LayerNotUsed=@n57
   ) {parameterName=Layer3}) {name=With_Sand, nodeColor=<rgba 1 1 0 1>}
  n103 = NamedRerouteUsage()
  NamedRerouteDeclaration(None=Multiply(A=@n103, B=NamedRerouteUsage())) {name=GM-BeachFoliage, desc="\"Beach Grass\"", nodeColor=<rgba 1 0 0.2578 1>}
  n104 = NamedRerouteUsage()
  n105 = Max(A=Max(A=NamedRerouteUsage(), B=NamedRerouteUsage()), B=NamedRerouteUsage())
  n106 = Max(A=Saturate(None=Subtract(A=@n104, B=@n105) {desc="Remove Sand, Asphalt, Cliff"}), B=NamedRerouteUsage()) {desc=Add JungleMud}
  n107 = NamedRerouteUsage()
  NamedRerouteDeclaration(None=Saturate(None=Subtract(A=@n106, B=@n107) {desc="Remove JungleFloor, JungleMoss"}) {desc=Final Dirt mask}) {name=LPM_Dirt, nodeColor=<rgba 0.2891 0 1 1>}
  NamedRerouteDeclaration(None=Multiply(A=NamedRerouteUsage(), B=@n103)) {name=GM-BeachFoliage_02, nodeColor=<rgba 0 0.1797 1 1>}
  NamedRerouteDeclaration(None=Constant4Vector() {constant=<rgba 0 0 0 1>}.RGBA) {name=BiomeMask-RGBA, nodeColor=<rgba 1 0 0.2344 1>}
  n108 = NamedRerouteUsage()
  n109 = NamedRerouteUsage()
  n110 = NamedRerouteUsage()
  n111 = Max(A=Max(A=Max(A=@n109, B=@n110), B=NamedRerouteUsage()), B=NamedRerouteUsage())
  NamedRerouteDeclaration(None=Saturate(None=Subtract(A=@n108, B=@n111) {desc="Remove Asphalt, JungleMud, JungleFloor, JungleMoss, Cliff"})) {name=LPM_Sand, nodeColor=<rgba 0.2891 0 1 1>}
  NamedRerouteDeclaration(None=OneMinus(None=NamedRerouteUsage()) {desc="White=Cliffs"}) {name=LPM_Stone, desc="White=Cliffs", nodeColor=<rgba 0.2891 0 1 1>}
  n112 = Max(A=NamedRerouteUsage(), B=NamedRerouteUsage())
  n113 = NamedRerouteUsage()
  n114 = Subtract(A=NamedRerouteUsage() {desc="Black=Cliffs"}, B=Max(A=Max(A=@n112, B=@n113), B=NamedRerouteUsage())) {desc="Remove Gravel,Mud,Sand, Asphalt, JungleMud"}
  NamedRerouteDeclaration(None=Max(A=Saturate(None=@n114), B=NamedRerouteUsage())) {name=LPM_Grass, nodeColor=<rgba 0.2891 0 1 1>}
  n115 = Saturate(None=LinearInterpolate(A=Constant() {r=-0.5}, B=Constant() {r=1.5}, Alpha=NamedRerouteUsage()))
  n116 = NamedRerouteUsage()
  n117 = Constant() {r=0.5}
  n118 = NamedRerouteUsage()
  NamedRerouteDeclaration(None=Multiply(
    A=MaterialFunctionCall(Transition Phase=@n115, Height Texture=@n116, Contrast=@n117) {materialFunction=&HeightLerp}.Alpha
    B=OneMinus(None=@n118)
   )) {name=GM-JungleFloor, nodeColor=<rgba 0.5 0 1 1>}
  NamedRerouteDeclaration(None=@n44.RegularHeightmap) {name=Cliff_Height, nodeColor=<rgba 1 0 0.3984 1>}
  n119 = Saturate(None=LinearInterpolate(A=Constant() {r=-1}, B=Constant() {r=2}, Alpha=NamedRerouteUsage()))
  n120 = NamedRerouteUsage()
  n121 = Constant() {r=1}
  n122 = NamedRerouteUsage()
  NamedRerouteDeclaration(None=Multiply(
    A=MaterialFunctionCall(Transition Phase=@n119, Height Texture=@n120, Contrast=@n121) {materialFunction=&HeightLerp}.Alpha
    B=OneMinus(None=@n122)
   )) {name=GM-CliffOvergrowth, nodeColor=<rgba 0.7109 0 1 1>}
  n123 = NamedRerouteUsage()
  n124 = Constant() {r=3}
  n125 = NamedRerouteDeclaration(None=Saturate(None=Multiply(A=@n123, B=@n124))) {name=GM-Asphalt, nodeColor=<rgba 0.2891 0 1 1>}
  n126 = Max(A=Max(A=NamedRerouteUsage(), B=NamedRerouteUsage()), B=NamedRerouteUsage())
  NamedRerouteDeclaration(None=Saturate(None=Subtract(A=@n125, B=@n126) {desc="Remove JungleMud, JungleFloor, JungleMoss, Cliff"})) {name=LPM_Asphalt, nodeColor=<rgba 0.2891 0 1 1>}
  NamedRerouteDeclaration(None=@n42.Resulting Height Map) {name=JungleMossHeight, nodeColor=<rgba 0.6928 0.156 1 1>}
  n127 = Saturate(None=LinearInterpolate(A=Constant() {r=-0.5}, B=Constant() {r=1}, Alpha=NamedRerouteUsage()))
  n128 = NamedRerouteUsage()
  n129 = Constant() {r=0.5}
  n130 = NamedRerouteUsage()
  NamedRerouteDeclaration(None=Multiply(
    A=MaterialFunctionCall(Transition Phase=@n127, Height Texture=@n128, Contrast=@n129) {materialFunction=&HeightLerp}.Alpha
    B=OneMinus(None=@n130)
   )) {name=GM-JungleMoss, nodeColor=<rgba 0.5 0 1 1>}
  n131 = LinearInterpolate(A=Constant() {r=-0.5}, B=Constant() {r=1.5}, Alpha=NamedRerouteUsage())
  NamedRerouteDeclaration(None=MaterialFunctionCall(Transition Phase=Saturate(None=@n131), Height Texture=NamedRerouteUsage(), Contrast=Constant() {r=0.5}) {materialFunction=&HeightLerp}.Alpha) {name=GM-JungleMud, nodeColor=<rgba 0.5 0 1 1>}
  NamedRerouteDeclaration(None=Max(A=NamedRerouteUsage(), B=NamedRerouteUsage())) {name="GM-JungleFloor&&Moss", nodeColor=<rgba 0.2891 0 1 1>}
  n132 = Max(A=NamedRerouteUsage(), B=NamedRerouteUsage())
  n133 = Constant() {r=2}
  NamedRerouteDeclaration(None=Saturate(None=Multiply(A=@n132, B=@n133))) {name="GM-Gravel&&Mud", nodeColor=<rgba 0.2891 0 1 1>}
  n134 = Subtract(A=NamedRerouteUsage(), B=NamedRerouteUsage())
  NamedRerouteDeclaration(None=Power(Base=Saturate(None=@n134)) {constExponent=3}) {name=GM-JungleMudMinusCliff, nodeColor=<rgba 0.2891 0 1 1>}
  NamedRerouteDeclaration(None=@n44.CliffMask) {name=CliffMask, nodeColor=<rgba 0.5 0 1 1>}
  NamedRerouteDeclaration(None=@n44.LandscapeGrassMeshMask) {name=GM-Cliff, nodeColor=<rgba 0.5 0 1 1>}
  n135 = GetMaterialAttributes(None=MaterialFunctionCall!(Material=@n44.Material) {materialFunction=/Game/Athena/Apollo/Maps/Landscape/VT/MaterialFunctions/TransformTangentToWorld}.Result) {attributeGetTypes [
    #2
    #3
    #4
    #5
   ]}
  n136 = SetMaterialAttributes!(MaterialAttributes, Base Color=@n135.Base Color, Specular=@n135.Base Color, Roughness=@n135.Base Color, Normal=@n135.Base Color) {attributeSetTypes [
    #2
    #3
    #4
    #5
   ]}
  n137 = Constant() {r=3}
  n138 = MaterialFunctionCall!(
   BiomeMask-RGBA=NamedRerouteUsage()
   CombinedLandscapeLayersMask=Saturate(None=Multiply(A=@n137, B=@n23))
   BypassUseRVT
   BypassBiomeColors=StaticBool() {value=true}
  ) {materialFunction=/Game/Athena/Asteria/Environments/Landscape/Materials/MaterialFunctions/MobileLayers/MF_Mobile_Asteria_Grass}
  n139 = MaterialFunctionCall(
   Material 1=@n138.MobileLandscapeMaterial
   Height Texture 1=@n138.MobileLandscapeMaterial
   Material 2=NamedRerouteDeclaration(None=@n87.Material) {name=MobileGravelMaterial, nodeColor=<rgba 1 1 0 1>}
   Height Texture 2=NamedRerouteDeclaration(None=@n87.Height) {name=MobileGravelHeight, nodeColor=<rgba 1 1 0 1>}
   Transistion Phase=@n86
  ) {materialFunction=&MF_BlendLandscapeLayers}
  n140 = MaterialFunctionCall(
   Material 1=@n139.BlendedMaterial
   Height Texture 1=@n139.BlendedMaterial
   Material 2=SetMaterialAttributes!(MaterialAttributes=@n85, Base Color=@n88.ColorizedGravelDiffuse_UEFN) {attributeSetTypes [#2]}
   Height Texture 2=NamedRerouteUsage()
   Transistion Phase=NamedRerouteUsage()
  ) {materialFunction=&MF_BlendLandscapeLayers}
  n141 = SetMaterialAttributes!(MaterialAttributes=@n140.BlendedMaterial, Opacity Mask=@n140.BlendedMaterial) {attributeSetTypes [#1]}
  n142 = MaterialFunctionCall!(CustomSeaLevel, BiomeMask-RGBA=NamedRerouteUsage(), BypassBiomeColors=StaticBool() {value=true}) {materialFunction=/Game/Athena/Asteria/Environments/Landscape/Materials/MaterialFunctions/MobileLayers/MF_Mobile_Asteria_Sand}
  n143 = MaterialFunctionCall(
   CurrentMaterial=@n141
   AddMaterial=SetMaterialAttributes(MaterialAttributes=@n142.Material, Opacity Mask=@n142.Material) {attributeSetTypes [#1]}
   LandscapeLayer=NamedRerouteUsage()
  ) {materialFunction=&MF_BlendLandscapeLayers_Opacity}
  n144 = LandscapeLayerSwitch(
   LayerUsed=SetMaterialAttributes(MaterialAttributes=@n143.BlendedMaterial, Opacity Mask=@n143.BlendedMaterial) {attributeSetTypes [#1]}
   LayerNotUsed=@n141
  ) {parameterName=Layer3}
  n145 = MaterialFunctionCall(
   CurrentMaterial=@n144
   AddMaterial=SetMaterialAttributes(MaterialAttributes=@n67.MobileLandscapeMaterial, Opacity Mask=@n67.MobileLandscapeMaterial) {attributeSetTypes [#1]}
   LandscapeLayer=NamedRerouteUsage()
  ) {materialFunction=&MF_BlendLandscapeLayers_Opacity}
  n146 = LandscapeLayerSwitch(
   LayerUsed=SetMaterialAttributes(MaterialAttributes=@n145.BlendedMaterial, Opacity Mask=@n145.BlendedMaterial) {attributeSetTypes [#1]}
   LayerNotUsed=@n144
  ) {parameterName=Layer4}
  n147 = GetMaterialAttributes(None=NamedRerouteUsage()) {attributeGetTypes [#2, #1]}
  n148 = Constant3Vector() {
   desc="The mobile material uses Layer2 (Mud) as Layer5. Use this to tint the JungleMud on mobile to differentiate it from the regular mud."
   constant=<rgba 0.75 0.5122 0.5078 1>
  }
  n149 = Saturate(None=Multiply(A=@n147.Base Color, B=@n148))
  n150 = Constant() {r=0.2}
  n151 = Constant() {r=0.65}
  n152 = Constant() {r=1.5}
  n153 = Saturate(None=Power(Base=Multiply(A=@n147.Opacity Mask, B=@n152), Exp=Constant() {r=3}))
  n154 = Multiply(A=OneMinus(None=Saturate(None=@n147.Opacity Mask)), B=Constant() {r=2})
  n155 = MaterialFunctionCall(
   CurrentMaterial=@n146
   AddMaterial=SetMaterialAttributes!(
    MaterialAttributes=@n147.MaterialAttributes
    Base Color=@n149
    Roughness=LinearInterpolate(A=@n150, B=@n151, Alpha=@n153)
    Opacity Mask=Saturate(None=@n154)
   ) {attributeSetTypes [#2, #4, #1]}
   LandscapeLayer=NamedRerouteUsage()
  ) {materialFunction=&MF_BlendLandscapeLayers_Opacity}
  n156 = LandscapeLayerSwitch(
   LayerUsed=SetMaterialAttributes!(MaterialAttributes=@n155.BlendedMaterial, Opacity Mask=@n155.BlendedMaterial) {attributeSetTypes [#1]}
   LayerNotUsed=@n146
  ) {parameterName=Layer5}
  n157 = ScalarParameter() {
   defaultValue=0.2
   desc=This only controls the scale of this texture on mobile platforms.
   group=JungleFloor
   parameterName=JungleFloor_MobileScale
  }
  n158 = TextureSample(UVs=Multiply(A=TextureCoordinate(), B=@n157), Tex=@n31) {samplerSource=SSM_Wrap_WorldGroupSettings}
  n159 = Constant() {r=0.97}
  n160 = Constant() {r=-0.6}
  n161 = SetMaterialAttributes!(
   MaterialAttributes
   Base Color=@n158.RGB
   Specular=Constant() {r=0.5}
   Roughness=Saturate(None=LinearInterpolate(A=@n159, B=@n160, Alpha=@n158.G))
   Normal=VertexNormalWS()
   Opacity Mask=@n158.RGB
  ) {attributeSetTypes [
    #2
    #3
    #4
    #5
    #1
   ]}
  n162 = MaterialFunctionCall(CurrentMaterial=@n156, AddMaterial=@n161, LandscapeLayer=NamedRerouteUsage(), UseContrast=@n34, Contrast=@n35) {materialFunction=&MF_BlendLandscapeLayers_Opacity}
  n163 = LandscapeLayerSwitch(
   LayerUsed=SetMaterialAttributes!(MaterialAttributes=@n162.BlendedMaterial, Opacity Mask=@n162.BlendedMaterial) {attributeSetTypes [#1]}
   LayerNotUsed=@n156
  ) {parameterName=Layer6}
  n164 = ScalarParameter() {
   defaultValue=0.15
   desc=This only controls the scale of this texture on mobile platforms.
   group=JungleMoss
   parameterName=JungleMoss_MobileScale
  }
  n165 = TextureSample(
   UVs=Multiply(A=TextureCoordinate(), B=@n164)
   Tex=TextureObject() {texture=/Game/Environments/Asteria/Sets/ZenGarden/Textures/T_Asteria_Zen_MossyGround_02}
  ) {samplerSource=SSM_Wrap_WorldGroupSettings}
  n166 = SetMaterialAttributes!(
   MaterialAttributes
   Base Color=@n165.RGB
   Specular=Constant() {r=0.5}
   Roughness=Constant() {r=1}
   Normal=VertexNormalWS()
   Opacity Mask=@n165.RGB
  ) {attributeSetTypes [
    #2
    #3
    #4
    #5
    #1
   ]}
  n167 = MaterialFunctionCall(CurrentMaterial=@n163, AddMaterial=@n166, LandscapeLayer=NamedRerouteUsage(), UseContrast=@n40, Contrast=@n41) {materialFunction=&MF_BlendLandscapeLayers_Opacity}
  n168 = GetMaterialAttributes(None=LandscapeLayerSwitch(
    LayerUsed=SetMaterialAttributes!(MaterialAttributes=@n167.BlendedMaterial, Opacity Mask=@n167.BlendedMaterial) {attributeSetTypes [#1]}
    LayerNotUsed=@n163
   ) {parameterName=Layer7}) {attributeGetTypes [#1]}
  n169 = MaterialFunctionCall(
   Material=SetMaterialAttributes!(MaterialAttributes=@n168.MaterialAttributes, Normal=VertexNormalWS()) {attributeSetTypes [#5]}
   Heightmap=@n168.Opacity Mask
  ) {materialFunction=&MF_CR_AsteriaCliffBlend}
  n170 = MaterialFunctionCall!(In=ShadingPathSwitch(Default=@n136, Mobile=@n169.Material), IsForLandscapeMaterial=StaticBool() {value=true}) {materialFunction=/Game/Creative/Materials/MaterialFunctions/MF_CreativeFortniteGameplay}
  n171 = LandscapeVisibilityMask()
  n172 = GetMaterialAttributes(None=MaterialFunctionCall!(In=SetMaterialAttributes!(MaterialAttributes=@n170.Result, Opacity Mask=@n171) {attributeSetTypes [#1]}) {materialFunction=/Game/Athena/Asteria/Environments/Landscape/Materials/MaterialFunctions/MF_Terrain_LumenModification}.Result) {attributeGetTypes [#2]}
  SetMaterialAttributes!(MaterialAttributes=@n172.MaterialAttributes, Base Color=LinearInterpolate(A=@n172.Base Color)) {attributeSetTypes [#2]}
  SetMaterialAttributes(MaterialAttributes=@n147.MaterialAttributes, Base Color=@n149) {attributeSetTypes [#2]}
  SetMaterialAttributes!(MaterialAttributes=@n169.Material, Shading Model=ShadingModel()) {attributeSetTypes [#6]}
  VectorParameter() {
   desc="The mobile material uses Layer2 (Mud) as Layer5. Use this to tint the JungleMud on mobile to differentiate it from the regular mud."
   group=JungleMud
   parameterName=JungleMud_MobileColorTint
   defaultValue=<rgba 0.75 0.5122 0.5078 1>
  }
  VertexNormalWS()
 ]
}
